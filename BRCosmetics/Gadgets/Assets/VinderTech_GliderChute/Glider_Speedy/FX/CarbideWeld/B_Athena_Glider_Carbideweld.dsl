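; Glider Blueprint for CarbideWeld that passes the player's selected emissive cosmetic variant to its Niagara trail effect.
; On full deployment and on visibility change, it looks up the active Emissive channel index among four property tags in the Glider loadout slot.
; It stores that index and writes it as the float "Variant" parameter on the first trail component.
; The visibility event passes a literal false to its parent instead of the incoming flag.
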(event EventOnGliderFullyDeployed
  (|CallParentFunction)
  (bind _index (CosmeticLoadout|GetVariantChannelActiveIndexforSlotActor "(TagName=\"LoadoutSlot.Glider\")" (Variables|FortPlayerParachute|GetPlayerPawn) "(TagName=\"Cosmetics.Variant.Channel.Emissive\")" (Utilities|Array|MakeArray "(TagName=\"Cosmetics.Variant.Property.Emissive1\")" "(TagName=\"Cosmetics.Variant.Property.Emissive2\")" "(TagName=\"Cosmetics.Variant.Property.Emissive3\")" "(TagName=\"Cosmetics.Variant.Property.Emissive4\")")))
  (Variables|Default|SetCurrent%20Variant _index)
  (Niagara|SetNiagaraVariable(Float) (Utilities|Array|Get(acopy) (Variables|FortPlayerParachute|GetTrailComponents)) "Variant" (Math|Conversions|ToFloat(Integer) (Variables|Default|GetCurrent%20Variant))))

(event EventOnChuteVisibilityChanged
  (params (bParachuteVisibility :type Boolean))
  (|CallParentFunction false)
  (bind _index (CosmeticLoadout|GetVariantChannelActiveIndexforSlotActor "(TagName=\"LoadoutSlot.Glider\")" (Variables|FortPlayerParachute|GetPlayerPawn) "(TagName=\"Cosmetics.Variant.Channel.Emissive\")" (Utilities|Array|MakeArray "(TagName=\"Cosmetics.Variant.Property.Emissive1\")" "(TagName=\"Cosmetics.Variant.Property.Emissive2\")" "(TagName=\"Cosmetics.Variant.Property.Emissive3\")" "(TagName=\"Cosmetics.Variant.Property.Emissive4\")")))
  (Variables|Default|SetCurrent%20Variant _index)
  (Niagara|SetNiagaraVariable(Float) (Utilities|Array|Get(acopy) (Variables|FortPlayerParachute|GetTrailComponents)) "Variant" (Math|Conversions|ToFloat(Integer) (Variables|Default|GetCurrent%20Variant))))

(fn ConstructionScript
  (|CallParentFunction))
